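# Lit material for the Raptor Eye Creative gameplay item, faking a deep glowing eye under a glassy surface with dirt on top.
# A dirt mask blends the tinted, tiled overlay and its normal against a bump-offset eye backdrop and flattened glass normal.
# Emissive layers parallax flame-noise masking and Fresnel interior and edge glows, suppressed under dirt.
# Attributes pass through MF_Pickup_Hologram; the bdrp_emis switch is left unconnected.
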Material /BRCosmetics/Creative/Items/Gameplay/Resources/Materials/M_CM_RaptorEye_01
graph {
 properties {bUseMaterialAttributes=true, pixelDepthOffsetMode=PDOM_Legacy, usedShadingModels=Default Lit}
 nodes [
  n1 = TextureCoordinate() {uTiling=8, vTiling=8}
  n2 = TextureSampleParameter2D(UVs=@n1) {group=Textures, parameterName=ForeGround_image, texture=/Game/Creative/Items/Gameplay/Resources/Textures/T_Raptor_eye_Dirt}
  n3 = VectorParameter() {group=property_adjust, parameterName=Color_Top_Tint, defaultValue=<rgba 0.8 0.1991 0.071 1>}
  n4 = TextureSampleParameter2D(UVs=@n1) {
   group=Textures
   parameterName=Top_masks_Emissive_Masks
   samplerType=SAMPLERTYPE_Masks
   texture=/Game/Creative/Items/Gameplay/Resources/Textures/T_Raptor_eye_Dirt_S
  }
  n5 = Add(A=LinearInterpolate(B=Multiply(A=@n2.RGB, B=@n3.RGB), Alpha=@n4.B), B=Constant() {r=0.2})
  n6 = BumpOffset(Coordinate=TextureCoordinate(), Height=ScalarParameter() {defaultValue=7.25, group=Offset_adjust, parameterName=BackOffset}) {heightRatio=0.1, referencePlane=8}
  n7 = Multiply(
   A=Multiply(
    A=TextureSampleParameter2D(UVs=@n6) {parameterName=BackDrop_Base_color, texture=/Game/Creative/Items/Gameplay/Resources/Textures/T_CP_Raptor_Eye_01_D}.RGB
    B=VectorParameter() {parameterName=Color_eye, defaultValue=<rgba 2 1.5174 0.5833 1>}.RGB
   )
   B=ScalarParameter() {defaultValue=0.15, group=property_adjust, parameterName=darken_back}
  )
  n8 = OneMinus(None=@n4.B)
  n9 = TextureSampleParameter2D(UVs=@n1) {
   group=Textures
   parameterName=PMSK_SMR
   samplerType=SAMPLERTYPE_Masks
   texture=&T_Raptor_eye_Dirt_S
  }
  n10 = ScalarParameter() {parameterName=metallic_adjust}
  n11 = ScalarParameter() {defaultValue=1, parameterName=Specular_adjust}
  n12 = ScalarParameter() {parameterName=Roughness_adjust}
  n13 = VectorParameter() {group=property_adjust, parameterName=Color_Tint, defaultValue=<rgba 0.5 0.5 0.5 1>}
  n14 = Multiply(A=@n4.B, B=Constant() {r=1})
  n15 = Constant() {r=0.001}
  n16 = ScalarParameter() {defaultValue=100, parameterName=Darken_mid_animation}
  n17 = TextureSampleParameter2D(UVs=@n6) {parameterName=Top_mask_Red, samplerType=SAMPLERTYPE_Masks, texture=/Game/Creative/Items/Gameplay/Resources/Textures/T_CP_Raptor_Eye_01_pmsk}
  n18 = TextureCoordinate()
  n19 = ScalarParameter() {defaultValue=1, group=Offset_adjust, parameterName=MidUVPositionX}
  n20 = ScalarParameter() {defaultValue=1, group=Offset_adjust, parameterName=MidUVPositionY}
  n21 = BumpOffset(
   Coordinate=Multiply(A=Panner(Coordinate=@n18), B=AppendVector(A=@n19, B=@n20))
   Height=ScalarParameter() {defaultValue=-0.1, group=Offset_adjust, parameterName=MidOffset}
  ) {heightRatio=0.1, referencePlane=3}
  n22 = Multiply(
   A=Multiply(
    A=@n17.G
    B=TextureSample(UVs=@n21) {samplerType=SAMPLERTYPE_LinearGrayscale, texture=/BRCosmetics/Weapons/FORT_Melee/Pickaxe_Cube_Ninja_Male/FX/T_FluidFlameTile}.G
   )
   B=VectorParameter() {parameterName=Animated_fire, defaultValue=<rgba 3 0.2884 0 1>}.RGB
  )
  n23 = OneMinus(None=@n17.R)
  n24 = PixelNormalWS()
  n25 = Multiply(
   A=VectorParameter() {group=InnerGlow, parameterName=Color-Emissive_InteriorGlow, defaultValue=<rgba 1.5 0.4636 0.1471 1>}.RGB
   B=MaterialFunctionCall(Exponent=ScalarParameter() {defaultValue=9, group=InnerGlow, parameterName=Fresnel-Exp_InteriorGlow}, Vertex / Pixel Normal=@n24) {materialFunction=&MF_Fresnel_FX}.Fresnel
  )
  n26 = TextureCoordinate() {uTiling=2}
  n27 = Multiply(
   A=AppendVector(
    A=ScalarParameter() {defaultValue=1, group=InnerGlow, parameterName=Bump_Offset_X}
    B=ScalarParameter() {defaultValue=1, group=InnerGlow, parameterName=Bump_Offset_Y}
   )
   B=@n26
  )
  n28 = ScalarParameter() {defaultValue=-4, group=InnerGlow, parameterName=BumpOffset_Height}
  n29 = ScalarParameter() {group=InnerGlow, parameterName=Fresnel Panner X}
  n30 = ScalarParameter() {group=InnerGlow, parameterName=Fresnel Panner Y}
  n31 = TextureSampleParameter2D(UVs=Panner(Coordinate=BumpOffset(Coordinate=@n27, Height=@n28) {heightRatio=0.2, referencePlane=0}, Speed=AppendVector(A=@n29, B=@n30))) {
   group=InnerGlow
   parameterName=Texture_InnerNoise
   samplerType=SAMPLERTYPE_LinearGrayscale
   texture=&T_FluidFlameTile
  }
  n32 = LinearInterpolate(Alpha=@n31.R) {constA=0.5}
  n33 = ScalarParameter() {defaultValue=2, group=InnerGlow, parameterName=OutNoiseX}
  n34 = ScalarParameter() {defaultValue=4, group=InnerGlow, parameterName=OutNoiseY}
  n35 = TextureSampleParameter2D(UVs=Multiply(A=AppendVector(A=@n33, B=@n34), B=@n26)) {
   group=InnerGlow
   parameterName=Texture_OuterNoise
   samplerType=SAMPLERTYPE_LinearGrayscale
   texture=&T_FluidFlameTile
  }
  n36 = ScalarParameter() {defaultValue=1.2, group=InnerGlow, parameterName=Edge-Brightness_InteriorGlow, sliderMin=-0.008}
  n37 = ScalarParameter() {defaultValue=4, group=InnerGlow, parameterName=Edge-Exp_InteriorGlow}
  n38 = Add(
   A=Multiply(A=Multiply(A=@n25, B=@n32), B=LinearInterpolate(Alpha=@n35.R) {constA=0.25})
   B=Multiply(
    A=VectorParameter() {group=InnerGlow, parameterName=Color-EdgeGlow_InteriorGlow, defaultValue=<rgba 10 3.1644 0.5415 1>}.RGB
    B=MaterialFunctionCall(Brightness=@n36, Exponent=@n37, Vertex / Pixel Normal=@n24) {materialFunction=&MF_Fresnel_FX}.Fresnel Inverted
   )
  )
  n39 = Constant()
  n40 = VectorParameter() {parameterName=LightColor, defaultValue=<rgba 0.1 0.0424 0.0339 1>}
  n41 = TextureSampleParameter2D(UVs=BumpOffset(Coordinate=TextureCoordinate(), Height=ScalarParameter() {defaultValue=0.65, group=property_adjust, parameterName=Emissive_Offset}) {heightRatio=0.2, referencePlane=1}) {parameterName=Top_mask_Red, samplerType=SAMPLERTYPE_Masks, texture=&T_CP_Raptor_Eye_01_pmsk}
  n42 = Constant() {r=0.8}
  n43 = Add(A=Multiply(A=LinearInterpolate(A=@n38, B=@n39) {constB=0}, B=@n40.RGB), B=Multiply(A=@n40.RGB, B=Multiply(A=@n41.B, B=@n42)))
  n44 = Add(A=LinearInterpolate(A=@n7, B=Multiply(A=@n15, B=@n16), Alpha=Multiply(A=@n22, B=@n23)), B=Multiply(A=@n43, B=OneMinus(None=@n17.R)))
  n45 = Constant()
  n46 = Multiply(
   A=Multiply(A=OneMinus(None=@n14), B=LinearInterpolate(A=@n44, B=@n45, Alpha=@n14))
   B=ScalarParameter() {defaultValue=1.25, parameterName=emissive_intens}
  )
  n47 = LinearInterpolate(B=Add(A=Add(A=@n13.RGB), B=@n46), Alpha=@n46)
  n48 = Multiply(A=TextureCoordinate(), B=ScalarParameter() {defaultValue=6, group=Glass, parameterName=Normal_TileRate})
  n49 = MaterialFunctionCall(
   Normal=TextureSample(UVs=Panner(Coordinate=@n48)) {samplerType=SAMPLERTYPE_Normal, texture=/Game/Creative/Items/Gameplay/Resources/Textures/T_RaptorEye_surface_N}.RGB
   Flatness=ScalarParameter() {defaultValue=1.15, group=Glass, parameterName=NormalFlatten}
  ) {materialFunction=&FlattenNormal}
  n50 = TextureSample(UVs=TextureCoordinate() {uTiling=16, vTiling=16}) {samplerType=SAMPLERTYPE_Normal, texture=/Game/Creative/Items/Gameplay/Resources/Textures/T_Raptor_eye_Dirt_n}
  MaterialFunctionCall!(In=MakeMaterialAttributes!(
    BaseColor=LinearInterpolate(A=@n5, B=@n7, Alpha=@n8)
    Metallic=Multiply(A=@n9.G, B=@n10)
    Specular=Multiply(A=@n9.R, B=@n11)
    Roughness=Multiply(A=@n9.B, B=@n12)
    Anisotropy
    EmissiveColor=@n47
    Opacity
    OpacityMask
    Normal=LinearInterpolate(A=@n49.Result, B=@n50.RGB, Alpha=@n4.B)
    Tangent
    WorldPositionOffset
    SubsurfaceColor
    ClearCoat
    ClearCoatRoughness
    AmbientOcclusion
    Refraction
    CustomizedUVs_0
    CustomizedUVs_1
    CustomizedUVs_2
    CustomizedUVs_3
    CustomizedUVs_4
    CustomizedUVs_5
    CustomizedUVs_6
    CustomizedUVs_7
    PixelDepthOffset
    ShadingModel
    Displacement
   ) {desc=Mask Exterior}) {materialFunction=/Game/Packages/Fortress_SharedMaterials/Base_Material_Functions/MF_Pickup_Hologram}
  Multiply(A=@n47, B=Constant() {r=0.5})
  StaticSwitchParameter(True=@n43, False=Constant()) {parameterName=bdrp_emis}
 ]
}
